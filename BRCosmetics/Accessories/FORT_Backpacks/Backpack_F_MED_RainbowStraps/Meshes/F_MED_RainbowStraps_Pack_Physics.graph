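# Physics asset for the RainbowStraps backpack cosmetic, providing bodies and joints for its dangling straps.
# Five capsules approximate the pack, and small capsules cover each strap bone and the two MSTR anchor bones.
# Straps 01 through 06 chain from one anchor with tight swing and twist limits, while strap 07 hangs from the other, with no joint between 06 and 07.
# Both anchors attach to the pack with free swing and twist.
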PhysicsAsset /BRCosmetics/Accessories/FORT_Backpacks/Backpack_F_MED_RainbowStraps/Meshes/F_MED_RainbowStraps_Pack_Physics
bodies {
 dyn_strap_01 [{
   shapeType=Capsule
   length=1
   radius=1.5
   shapeName=dyn_strap_01_capsule
   center=<2.3622 -0.0001 -0.1179>
   rotation=<rot 90 -144.7357 -144.736>
  }]
 dyn_strap_01_MSTR [{shapeType=Capsule, length=0.505, radius=0.505, shapeName=dyn_strap_01_MSTR_capsule, rotation=<rot 0 -90 0>}]
 dyn_strap_02 [{shapeType=Capsule, length=2.5, radius=1.5, shapeName=dyn_strap_02_capsule, center=<3.3051 0 -0.0712>, rotation=<rot 90 0 0>}]
 dyn_strap_03 [{shapeType=Capsule, length=3, radius=2, shapeName=dyn_strap_03_capsule, center=<3.8257 -0.5533 -0.2141>, rotation=<rot 90 0 0>}]
 dyn_strap_04 [{shapeType=Capsule, length=5, radius=2, shapeName=dyn_strap_03_capsule, center=<4.622 0.0906 0.932>, rotation=<rot 80.0008 -179.9998 -180>}]
 dyn_strap_05 [{shapeType=Capsule, length=1, radius=1, shapeName=dyn_strap_05_capsule, center=<-0.9343 1.0876 -0.1445>, rotation=<rot 90 -90 -50>}]
 dyn_strap_06 [{shapeType=Capsule, length=3, radius=2, shapeName=dyn_strap_06_capsule, center=<3.5661 -0.6468 0.0118>, rotation=<rot 90 176.4237 -173.5763>}]
 dyn_strap_07 [{shapeType=Capsule, length=5, radius=2, shapeName=dyn_strap_07_capsule, center=<4.4214 -0.6744 0.0645>, rotation=<rot 90 2.3859 12.386>}]
 dyn_strap_07_MSTR [{shapeType=Capsule, length=0.505, radius=0.505, shapeName=dyn_strap_07_MSTR_capsule, rotation=<rot 0 -90 0>}]
 pack [
  {shapeType=Capsule, length=15.4482, radius=5.4803, center=<2.8138 -6.8229 1.8663>, rotation=<rot 77.6418 -52.2768 37.0751>}
  {shapeType=Capsule, length=19.5, radius=5.5, shapeName=pack_capsule, center=<2.5708 -6.7342 -4.6001>, rotation=<rot 79.9999 -89.9994 0>}
  {shapeType=Capsule, length=20.5, radius=6.5, shapeName=pack_capsule, center=<-0.51 -5.7293 -12.1945>, rotation=<rot 79.9999 -89.9998 0>}
  {shapeType=Capsule, length=17.5, radius=6, shapeName=pack_capsule, center=<0.8256 1.9739 0>, rotation=<rot 0.0003 90 90>}
  {shapeType=Capsule, length=15, radius=7, shapeName=pack_capsule, center=<1.4206 3.2568 -11.5788>, rotation=<rot 0.0002 90 90>}
 ]
}
constraints [
 {bone1Name=dyn_strap_02, bone2Name=dyn_strap_01, swing1LimitDegrees=30, swing2LimitDegrees=20, twistLimitDegrees=0}
 {bone1Name=dyn_strap_01, bone2Name=dyn_strap_01_MSTR, swing1LimitDegrees=10, swing2LimitDegrees=10, twistLimitDegrees=10}
 {bone1Name=dyn_strap_03, bone2Name=dyn_strap_02, swing1LimitDegrees=30, swing2LimitDegrees=20, twistLimitDegrees=0}
 {bone1Name=dyn_strap_05, bone2Name=dyn_strap_04, swing1LimitDegrees=2, swing2LimitDegrees=2, twistLimitDegrees=2}
 {bone1Name=dyn_strap_06, bone2Name=dyn_strap_05, swing1LimitDegrees=8, swing2LimitDegrees=8, twistLimitDegrees=8}
 {bone1Name=dyn_strap_07, bone2Name=dyn_strap_07_MSTR, swing1LimitDegrees=40, swing2LimitDegrees=8, twistLimitDegrees=5}
 {
  bone1Name=dyn_strap_01_MSTR
  bone2Name=pack
  swing1LimitDegrees=45
  swing1Motion=Free
  swing2LimitDegrees=45
  swing2Motion=Free
  twistLimitDegrees=45
  twistMotion=Free
 }
 {
  bone1Name=dyn_strap_07_MSTR
  bone2Name=pack
  swing1LimitDegrees=45
  swing1Motion=Free
  swing2LimitDegrees=45
  swing2Motion=Free
  twistLimitDegrees=45
  twistMotion=Free
 }
 {bone1Name=dyn_strap_04, bone2Name=dyn_strap_03, swing1LimitDegrees=30, swing2LimitDegrees=8, twistLimitDegrees=5}
]
